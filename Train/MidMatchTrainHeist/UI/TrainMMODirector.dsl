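; Dynamic UI director for the Train MidMatchTrainHeist event that conditionally loads its HUD scene.
; On BeginPlay it gets the Fort global UI context from the owning local player's controller and adds TrainMMOScene only if that context reports being in zone.
; Tick and the construction script are empty.
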(event EventBeginPlay
  (if (GlobalUIContext|IsinZone (PlayerController|LocalPlayerSubsystems|GetFortGlobalUIContext (DynamicUIDirector|GetOwningLocalPlayerController)))
    (DynamicUIDirector|AddScene "/Train/MidMatchTrainHeist/UI/TrainMMOScene.TrainMMOScene")))

(event EventTick
  (params (DeltaSeconds :type Float%20%28single-precision%29)))

(fn ConstructionScript)
