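# Scene marker widget for the Train mid-match heist that shows an activated objective's countdown through an embedded WBP_TrainMMO_Progress.
# An ObjectiveActivated event message supplies the objective and activation time, offsetting the countdown by elapsed server time.
# Looping timers update progress and fade the marker by the view pawn's horizontal distance, using only SF_MinViewableDistance as threshold.
# It collapses while spectating non-teammates, and completion clears timers and listener, animating if visible.
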WidgetBlueprint /Train/MidMatchTrainHeist/UI/UI_TrainMMO_ProgressSceneMarker
graphs [
 CacheAssignedLocalPlayerPawn {nodes [
   Inputs!() -> (In:Exec)
   > n1 = Get!(execute:Exec) -> (then:Exec, else:Exec, CachedLocalPawn:Pawn Object Reference)
   Outputs!(
    Valid:Exec=[@n5.then, @n1.then]
    CachedPawn:Pawn Object Reference=Get CachedLocalPawn() -> (CachedLocalPawn:Pawn Object Reference)
    Invalid:Exec=[@n4.CastFailed, @n3.Is Not Valid]
   )
   n2 = GetOwningPlayer()
   n3 = Is Valid(exec=@n1.else, InputObject=@n2)
   > n4 = Cast To Pawn!(execute, Object=GetViewTarget!(self:Controller Object Reference=@n2) -> (ReturnValue:Actor Object Reference)) -> (then, CastFailed, AsPawn:Pawn Object Reference) {is=title K2Node_DynamicCast}
   > n5 = Set CachedLocalPawn(CachedLocalPawn:Pawn Object Reference=@n4) -> (Output_Get:Pawn Object Reference)
  ]}
 CalculateElapsedTime {nodes [
   n1 = CalculateElapsedTime!() -> (then, ActivationTime:double=~) {is=title K2Node_FunctionEntry}
   n2 = float - float(A=GetServerWorldTimeSeconds!() -> (ReturnValue:float), B=@n1)
   > n3 = Branch(Condition="float < float"(A=@n2))
   > Set ElapsedTime(ElapsedTime:double=0) -> (Output_Get:double)
   Set ElapsedTime(execute=@n3.else, ElapsedTime=@n2)
  ]}
 EventGraph {nodes [
   n1 = StartVisualTimer!() -> (OutputDelegate, then, TimerDuration:double=~) {is=title K2Node_CustomEvent}
   > n2 = Set TotalTimerDuration(TotalTimerDuration:double=@n1.TimerDuration) -> (Output_Get:double)
   > n3 = Set Timer by Event(Delegate=Create Event(), Time=float - float(A=@n2, B=Get ElapsedTime() -> (ElapsedTime:double)), bLooping=false)
   > Set Timer_ObjectiveTime(Timer_ObjectiveTime:Timer Handle Structure=@n3) -> (Output_Get:Timer Handle Structure)
   > n4 = Set Timer by Event(Delegate=Create Event(), Time=Get TimerPollingUpdateInterval() -> (TimerPollingUpdateInterval:double), bLooping=true)
   > Set Timer_UpdateCircleTime(Timer_UpdateCircleTime:Timer Handle Structure=@n4) -> (Output_Get:Timer Handle Structure)
   PollingUpdateVisualTimer() {is=title K2Node_CustomEvent}
   n5 = Get Timer Elapsed Time by Handle!(Handle:Timer Handle Structure=Get Timer_ObjectiveTime() -> (Timer_ObjectiveTime:Timer Handle Structure)) -> (ReturnValue:float) {is=title K2Node_CallFunction}
   n6 = Get ElapsedTime()
   > UpdateProgress!(
    execute
    self:WBP Train MMO Progress Object Reference=Get WBP_TrainMMO_Progress() -> (WBP_TrainMMO_Progress:WBP Train MMO Progress Object Reference)
    ProgressPercent:double=SafeDivide(A=float + float(A=@n5, B=@n6), B=Get TotalTimerDuration() -> (TotalTimerDuration:double))
   ) -> (then)
   VisualTimerComplete() {is=title K2Node_CustomEvent}
   > n7 = Branch(Condition=Get bIsWidgetVisibleAtDistance() -> (bIsWidgetVisibleAtDistance:bool))
   n8 = Get WBP_TrainMMO_Progress()
   > Bind Event to Complete Anim Finished(self:WBP Train MMO Progress Object Reference=@n8, Delegate=Create Event()) {is=title K2Node_AddDelegate}
   > n9 = PlayCompleteAnimation(self:WBP Train MMO Progress Object Reference=@n8)
   Deinitialize(execute=[@n9.then, @n7.else])
   > Set bIsWidgetVisibleAtDistance(bIsWidgetVisibleAtDistance:bool=false) -> (Output_Get:bool)
   n10 = ActivateObjectiveWidget!() -> (OutputDelegate, then, ObjectiveObject:BGA Train MMO Interaction Parent Object Reference) {is=title K2Node_CustomEvent}
   > Set Objective Object(Objective Object:BGA Train MMO Interaction Parent Object Reference=@n10.ObjectiveObject) -> (Output_Get:BGA Train MMO Interaction Parent Object Reference)
   > n11 = Branch(Condition=NOT Boolean(A=Get SpectatingNonTeammate() -> (SpectatingNonTeammate:bool)))
   > n12 = SetVisibility(InVisibility=SelfHitTestInvisible)
   n13 = Cache Assigned Local Player Pawn!(In:Exec=[@n12.then, @n11.else]) -> (Valid:Exec, CachedPawn:Pawn Object Reference, Invalid:Exec) {is=title K2Node_MacroInstance}
   > n14 = Sequence()
   > n15 = Get!(execute) -> (then, else, Objective Object:BGA Train MMO Interaction Parent Object Reference)
   > n16 = Viewable Distance from Player!(
    In:Exec
    DistanceFromOwningPlayer:double=GetHorizontalDistanceTo!(self:Actor Object Reference=@n13, OtherActor:Actor Object Reference=@n15) -> (ReturnValue:float)
   ) -> (DeactivateWidget:Exec, ActivateWidget:Exec, HoldCurrent:Exec) {is=title K2Node_MacroInstance}
   OnObjectiveActivated(execute=@n16.ActivateWidget, self:WBP Train MMO Progress Object Reference=Get WBP_TrainMMO_Progress())
   n17 = Set Timer by Event(execute=@n14.then_1, Delegate=Create Event(), Time=0.2, bLooping=true)
   > Set Timer_DistanceUpdate(Timer_DistanceUpdate:Timer Handle Structure=@n17) -> (Output_Get:Timer Handle Structure)
   > StartVisualTimer!(
    execute
    self
    TimerDuration:double=GetValueAtLevel!(Input:"Scalable Float Structure (by ref)"=Get SF_ObjectiveDuration() -> (SF_ObjectiveDuration:Scalable Float Structure), Level:float=0) -> (ReturnValue:float)
   ) -> (then) {is=title K2Node_CallFunction}
   n18 = SetControlVisuals!() -> (OutputDelegate, then, ControlAffiliation:EFortTeamAffiliation Enum=Friendly) {is=title K2Node_CustomEvent}
   > SetObjectiveAffiliation!(
    execute
    self:WBP Train MMO Progress Object Reference=Get WBP_TrainMMO_Progress()
    ObjectiveAffiliation:EFortTeamAffiliation Enum=@n18.ControlAffiliation
   ) -> (then)
   PollingUpdateDistanceVisuals() {is=title K2Node_CustomEvent}
   > n19 = Cache Assigned Local Player Pawn()
   > n20 = Get()
   n21 = GetHorizontalDistanceTo(self=@n19, OtherActor=@n20)
   > SetTextDistance!(execute, self:WBP Train MMO Progress Object Reference=Get WBP_TrainMMO_Progress(), Display Distance:bool=false, Distance:double=@n21) -> (then)
   > n22 = Viewable Distance from Player(DistanceFromOwningPlayer=@n21)
   > Set bIsWidgetVisibleAtDistance(bIsWidgetVisibleAtDistance=false)
   n23 = Get WBP_TrainMMO_Progress()
   > PlayFadeOutAnimation(self:WBP Train MMO Progress Object Reference=@n23)
   Set bIsWidgetVisibleAtDistance(execute=@n22.ActivateWidget, bIsWidgetVisibleAtDistance=true)
   > PlayFadeInAnimation(self:WBP Train MMO Progress Object Reference=@n23)
   n24 = OnViewTargetChanged!() -> (
    OutputDelegate
    then
    PlayerController:Fort Player Controller Object Reference
    OldViewTarget:Actor Object Reference
    NewViewTarget:Actor Object Reference
   ) {is=title K2Node_CustomEvent}
   > n25 = Cast To FortPlayerControllerAthena(Object=@n24.PlayerController)
   > n26 = Cast To FortPlayerPawn(Object=@n24.NewViewTarget)
   n27 = GetActorTeam!(Actor:Actor Object Reference=@n26) -> (ReturnValue:Byte=0)
   n28 = GetActorTeam(Actor=@n25) -> (ReturnValue=0)
   > n29 = Set SpectatingNonTeammate(SpectatingNonTeammate:bool=AND Boolean(
     A=GetIsSpectating!(self:Fort Player Controller Athena Object Reference=@n25) -> (ReturnValue:bool)
     B="Not Equal (Byte)"(A:Byte=@n28, B:Byte=@n27) -> (ReturnValue:bool=~) {is=title K2Node_PromotableOperator}
    )) -> (Output_Get:bool)
   > n30 = Branch(Condition=@n29)
   > SetVisibility(InVisibility=Collapsed)
   Set CachedLocalPawn(execute=@n30.else)
   > n31 = Cache Assigned Local Player Pawn()
   n32 = Get Objective Object() -> (Objective Object:BGA Train MMO Interaction Parent Object Reference)
   Is Valid(exec=[@n31.Invalid, @n31.Valid], InputObject=@n32)
   > Branch(Condition="Equal (Enum)"(
     A:ETrainMMOInteractObjectiveState Enum=Get ECurrentObjectiveState_RepNotify!(self:BGA Train MMO Interaction Parent Object Reference=@n32) -> (ECurrentObjectiveState_RepNotify:ETrainMMOInteractObjectiveState Enum=NewEnumerator0)
     B:ETrainMMOInteractObjectiveState Enum=NewEnumerator3
    ))
   > SetVisibility(InVisibility=SelfHitTestInvisible)
   Deinitialize() {is=title K2Node_CustomEvent}
   > Clear and Invalidate Timer by Handle(Handle=Get Timer_DistanceUpdate() -> (Timer_DistanceUpdate:Timer Handle Structure))
   > Clear and Invalidate Timer by Handle(Handle=Get Timer_ObjectiveTime())
   > Clear and Invalidate Timer by Handle(Handle=Get Timer_UpdateCircleTime() -> (Timer_UpdateCircleTime:Timer Handle Structure))
   > StopListeningToEvent(self:Async Action Start Listening To Event Object Reference=Get Async Task Ref() -> (Async Task Ref:Async Action Start Listening To Event Object Reference))
   Event Destruct()
   > Parent: Destruct()
   > Deinitialize() {is=title K2Node_CallFunction}
   > n33 = Cast To FortPlayerController(Object=GetOwningPlayer())
   > Unbind Event from On View Target Changed(self:Fort Player Controller Object Reference=@n33, Delegate=Create Event()) {is=title K2Node_RemoveDelegate}
   Event OnInitialized()
   > n34 = Cast To FortPlayerController(Object=GetOwningPlayer())
   > n35 = Bind Event to On View Target Changed(self:Fort Player Controller Object Reference=@n34, Delegate=Create Event()) {is=title K2Node_AddDelegate}
   SetVisibility(execute=[@n35.then, @n34.CastFailed], InVisibility=Collapsed)
   > UpdateProgress(self=Get WBP_TrainMMO_Progress(), ProgressPercent=0)
   > n36 = StartListeningToEventBackwardCompatible!(
    execute:Exec
    Channel:Event Message Tag Structure=(TagName="EventMessage.Athena.TrainMMO.ObjectiveActivated")
    Target:Gameplay Event Router Component Object Reference=GetGlobalEventRouter()
    EventType:Script Struct Object Reference=/Train/MidMatchTrainHeist/TrainMMO_ObjectiveActivated
    Context:Object Reference
    EventBubblingRule:EEventBubblingRule Enum
   ) -> (
    then:Exec
    AsyncTaskProxy:Async Action Start Listening To Event Object Reference
    OnEventReceived:Exec
    Payload:Train MMO Objective Activated Structure
   ) {is=title K2Node_StartListeningToEventAsyncAction}
   > Set Async Task Ref(Async Task Ref:Async Action Start Listening To Event Object Reference=@n36) -> (Output_Get:Async Action Start Listening To Event Object Reference)
   SetControlVisuals!(execute=@n36.OnEventReceived, self, ControlAffiliation:EFortTeamAffiliation Enum=Neutral) -> (then) {is=title K2Node_CallFunction}
   > Set bIsWidgetVisibleAtDistance(bIsWidgetVisibleAtDistance=false)
   n37 = Break Train MMO Objective Activated!(TrainMMO_ObjectiveActivated:"Train MMO Objective Activated Structure (by ref)"=@n36.Payload) -> (
    ActiveObjective_7_#1:BGA Train MMO Interaction Parent Object Reference
    TeamAffiliation_5_#2:EFortTeamAffiliation Enum
    ActivationTime_10_#3:double=~
   ) {is=title K2Node_BreakStruct}
   > CalculateElapsedTime!(execute, self, ActivationTime:double=@n37.ActivationTime_10_#3) -> (then) {is=title K2Node_CallFunction}
   > ActivateObjectiveWidget!(execute, self, ObjectiveObject:BGA Train MMO Interaction Parent Object Reference=@n37) -> (then) {is=title K2Node_CallFunction}
  ]}
 OnProgressWidgetCompleteAnimFinished {nodes [
   OnProgressWidgetCompleteAnimFinished() {is=title K2Node_FunctionEntry}
   > SetVisibility(InVisibility=Collapsed)
  ]}
 ViewableDistanceFromPlayer {nodes [
   n1 = Inputs!() -> (In, DistanceFromOwningPlayer:double=~)
   n2 = GetValueAtLevel(Input=Get SF_MinViewableDistance() -> (SF_MinViewableDistance:Scalable Float Structure), Level=0)
   > n3 = Branch(Condition=AND Boolean(A="float > float"(A=@n1), B="float < float"(A=@n1, B=@n2)))
   > n4 = Branch(Condition=NOT Boolean(A=Get bIsWidgetVisibleAtDistance()))
   Outputs!(DeactivateWidget:Exec=@n5.then, ActivateWidget:Exec=@n4.then, HoldCurrent:Exec=[@n4.else, @n5.else])
   n5 = Branch(execute=@n3.else, Condition=Get bIsWidgetVisibleAtDistance())
  ]}
]
variables [
 Async Task Ref {type=AsyncAction_StartListeningToEvent, default=null}
 bEvaluateIsWidgetVisible {type=boolean}
 bIsWidgetVisibleAtDistance {type=boolean}
 CachedLocalPawn {type=Pawn, default=null}
 ElapsedTime {type=number}
 Objective Object {type=BGA_Train_MMO_Interaction_Parent_C, default=null}
 SF_MaxViewableDistance {
  type=ScalableFloat
  default {
   registryType=None
   value=1
   curve {curveTable=/Train/DataTables/MidMatchTrainHeistGameData, rowName=Train.MidMatchObjective.UI.MaxDistanceToUnhideProgressWidgetInUU}
  }
 }
 SF_MinViewableDistance {
  type=ScalableFloat
  default {
   registryType=None
   value=1
   curve {curveTable=&MidMatchTrainHeistGameData, rowName=Train.MidMatchObjective.UI.MinDistanceToHideProgressWidgetInUU}
  }
 }
 SF_ObjectiveDuration {
  type=ScalableFloat
  default {
   registryType=None
   value=1
   curve {curveTable=&MidMatchTrainHeistGameData, rowName=Train.MidMatchObjective.Interaction.ObjectiveCountdown}
  }
 }
 SpectatingNonTeammate {type=boolean}
 Text_ObjectiveControlled {type=string, default=Unlocking...}
 Text_ObjectiveLost {type=string, default=Claim}
 Timer_DistanceUpdate {type=TimerHandle}
 Timer_ObjectiveTime {type=TimerHandle}
 Timer_UpdateCircleTime {type=TimerHandle}
 TimerPollingUpdateInterval {type=number, default=0.2}
 TotalTimerDuration {type=number, default=-1}
]
widgets {
 info {inheritedWidgetCount=0, parentClass=&UserWidget, rootWidgetClass=&Overlay, widgetCount=4}
 widgets [
  Overlay_Root {widgetClassPath=&Overlay, widget=Overlay_Root}
  Overlay_Main {widgetClassPath=&Overlay, parent=Overlay_Root, widget=Overlay_Main, slot {class=OverlaySlot}}
  SizeBox_0 {
   widgetClassPath=&SizeBox
   parent=Overlay_Main
   widget=SizeBox_0
   properties {bOverride_HeightOverride=true, heightOverride=250}
   slot {class=OverlaySlot}
  }
  WBP_TrainMMO_Progress {
   widgetClassPath=&/WBP_TrainMMO_Progress
   bIsVariable=true
   parent=SizeBox_0
   widget=WBP_TrainMMO_Progress
   slot {class=SizeBoxSlot, properties {horizontalAlignment=HAlign_Left, verticalAlignment=VAlign_Top}}
  }
 ]
}
